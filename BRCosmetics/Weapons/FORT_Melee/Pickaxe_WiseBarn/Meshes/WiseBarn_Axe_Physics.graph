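# Physics asset for the WiseBarn pickaxe axe mesh, setting up simulation for its crystal bones.
# It has two capsule bodies, a tiny one on base_crystal and a larger one on dyn_crystal.
# A single constraint ties dyn_crystal to base_crystal with a limited swing on one axis and zero limits on the other swing and twist, giving hinge-like sway.
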PhysicsAsset /BRCosmetics/Weapons/FORT_Melee/Pickaxe_WiseBarn/Meshes/WiseBarn_Axe_Physics
bodies {
 base_crystal [{shapeType=Capsule, length=0.505, radius=0.505, shapeName=base_crystal_capsule, rotation=<rot 0 -90 0>}]
 dyn_crystal [{shapeType=Capsule, length=2, radius=3, shapeName=dyn_crystal_capsule}]
}
constraints [{bone1Name=dyn_crystal, bone2Name=base_crystal, swing1LimitDegrees=15, swing2LimitDegrees=0, twistLimitDegrees=0}]
